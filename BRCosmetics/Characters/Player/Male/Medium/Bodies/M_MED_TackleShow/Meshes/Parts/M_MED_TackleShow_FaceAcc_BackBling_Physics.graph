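# Physics asset for the M_MED_TackleShow character's face accessory/back bling part, setting up a long dynamic hair rig with body collision shapes.
# A twelve-segment center capsule chain hangs from a head-locked base with tight angular limits, branching into left and right strands.
# Extra constraints tie lower right segments back to the center chain.
# Torso, limb and head capsules use free-angle constraints, some linked to upper hair, alongside a large ground-collider sphere.
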PhysicsAsset /BRCosmetics/Characters/Player/Male/Medium/Bodies/M_MED_TackleShow/Meshes/Parts/M_MED_TackleShow_FaceAcc_BackBling_Physics
bodies {
 calf_l [{
   shapeType=Capsule
   length=41.1557
   radius=18.3976
   shapeName=calf_l_capsule
   center=<-25.8993 1.8462 0.7719>
   rotation=<rot -89.8496 -174.7837 -0.0011>
  }]
 calf_r [{
   shapeType=Capsule
   length=41.1557
   radius=18.3976
   shapeName=calf_r_capsule
   center=<25.8993 -1.8462 -0.7719>
   rotation=<rot -89.8496 -174.7837 179.9989>
  }]
 deform_ground_collider [{shapeType=Sphere, radius=300, shapeName=root_sphere, center=<0 -45 -304>}]
 dyn_hair_01_c [{
   shapeType=Capsule
   length=3.6352
   radius=2.3118
   shapeName=dyn_hair_01_c_capsule
   center=<4.7734 0 -0.1068>
   rotation=<rot 76.0756 -180 -180>
  }]
 dyn_hair_01_l [{shapeType=Capsule, length=12, radius=3, shapeName=dyn_hair_01_l_capsule, center=<7 0 0>, rotation=<rot 90 0 0>}]
 dyn_hair_01_r [{shapeType=Capsule, length=10, radius=3, shapeName=dyn_hair_01_r_capsule, center=<5 0 0>, rotation=<rot 90 0 0>}]
 dyn_hair_02_c [{shapeType=Capsule, length=5.3275, radius=4.6725, shapeName=dyn_hair_02_c_capsule, center=<7.5113 0 -0.9632>, rotation=<rot 82.5455 180 180>}]
 dyn_hair_02_l [{shapeType=Capsule, length=15, radius=3, shapeName=dyn_hair_02_l_capsule, center=<9.099 0.4759 0.2065>, rotation=<rot 90 6.977 0>}]
 dyn_hair_02_r [{shapeType=Capsule, length=12, radius=7, shapeName=dyn_hair_02_r_capsule, center=<6.9847 0 -0.3465>, rotation=<rot 87.7034 0.0006 0.0006>}]
 dyn_hair_03_c [{shapeType=Capsule, length=8.408, radius=4.8961, shapeName=dyn_hair_03_c_capsule, center=<5.8596 -1.2069 -0.5066>, rotation=<rot 0 0 90>}]
 dyn_hair_03_r [{shapeType=Capsule, length=12, radius=5, shapeName=dyn_hair_03_r_capsule, center=<7 0 0>, rotation=<rot 84.5755 0 0>}]
 dyn_hair_04_c [{shapeType=Capsule, length=14.2751, radius=5.3521, shapeName=dyn_hair_04_c_capsule, center=<5 -0.9133 0.3997>, rotation=<rot 0 0 90>}]
 dyn_hair_04_r [{
   shapeType=Capsule
   length=8.6359
   radius=7.3641
   shapeName=dyn_hair_04_r_capsule
   center=<8.914 -0.8659 -0.1475>
   rotation=<rot 87.8265 -179.9997 -173.3641>
  }]
 dyn_hair_05_c [{
   shapeType=Capsule
   length=19.7281
   radius=6.1392
   shapeName=dyn_hair_05_c_capsule
   center=<5 -1.3526 0.3968>
   rotation=<rot 0.0003 0 86.2541>
  }]
 dyn_hair_06_c [{shapeType=Capsule, length=21.1103, radius=6.882, shapeName=dyn_hair_06_c_capsule, center=<5 -0.0406 0.8126>, rotation=<rot 0.0001 0 87.1417>}]
 dyn_hair_07_c [{shapeType=Capsule, length=23.2488, radius=6.145, shapeName=dyn_hair_07_c_capsule, center=<5 1.7486 0.5407>, rotation=<rot 0.0001 0 87.277>}]
 dyn_hair_08_c [{shapeType=Capsule, length=18.885, radius=6.115, shapeName=dyn_hair_08_c_capsule, center=<5.5284 0 0>, rotation=<rot 0.0001 0 85.8972>}]
 dyn_hair_09_c [{shapeType=Capsule, length=20.7437, radius=6.9271, shapeName=dyn_hair_09_c_capsule, center=<6.1679 2.3782 0>, rotation=<rot 0 0 90>}]
 dyn_hair_10_c [{shapeType=Capsule, length=14.438, radius=7.7704, shapeName=dyn_hair_10_c_capsule, center=<6.122 1.3586 -1.0757>, rotation=<rot 0 0 90>}]
 dyn_hair_11_c [{shapeType=Capsule, length=6.3865, radius=10.1821, shapeName=dyn_hair_11_c_capsule, center=<6.8196 0 -3.3405>, rotation=<rot 0 0 90>}]
 dyn_hair_12_c [{
   shapeType=Capsule
   length=7.1075
   radius=7.3271
   shapeName=dyn_hair_12_c_capsule
   center=<7.3303 -3.921 -2.2192>
   rotation=<rot 0.0001 64.9474 83.6292>
  }]
 dyn_hair_base [{shapeType=Capsule, length=1, radius=1, shapeName=dyn_hair_base_capsule}]
 head [{shapeType=Capsule, length=2.9312, radius=19.0688, center=<10.3087 4.4189 0>, rotation=<rot 90 -30.0002 0>}]
 pelvis [{shapeType=Capsule, length=16.5457, radius=13.4543, shapeName=pelvis_capsule, center=<2.6354 2.4337 0>}]
 spine_01 [{shapeType=Capsule, length=11.6059, radius=12.3941, shapeName=spine_01_capsule, center=<3.5285 3.3674 0>}]
 spine_02 [{shapeType=Capsule, length=10.1211, radius=11.8789, shapeName=spine_02_capsule, center=<2.7463 4.0731 0>}]
 spine_03 [{shapeType=Capsule, length=11.4621, radius=11.5379, shapeName=spine_03_capsule, center=<3.8973 3.0597 0>}]
 spine_04 [{shapeType=Capsule, length=12.1831, radius=12.8169, shapeName=spine_04_capsule, center=<3.7492 0.5493 0>}]
 spine_05 [{shapeType=Capsule, length=52.9339, radius=13.4771, shapeName=spine_05_capsule, center=<0 0.5201 0>}]
 thigh_l [{shapeType=Capsule, length=32.6757, radius=9.9414, shapeName=thigh_l_capsule, center=<-18.9041 0.5728 0>, rotation=<rot -90 177.0438 0>}]
 thigh_r [{shapeType=Capsule, length=32.6757, radius=9.9414, shapeName=thigh_r_capsule, center=<18.9041 -0.5728 0>, rotation=<rot -90 -2.9562 0>}]
 upperarm_l [{
   shapeType=Capsule
   length=25
   radius=8
   shapeName=upperarm_l_capsule
   center=<17.7242 0.1652 -0.3788>
   rotation=<rot -87.5301 27.5982 153.7988>
  }]
 upperarm_r [{
   shapeType=Capsule
   length=25
   radius=8
   shapeName=upperarm_r_capsule
   center=<-17.7242 -0.1652 0.3788>
   rotation=<rot -87.5301 27.5982 -26.2012>
  }]
}
constraints [
 {
  bone1Name=dyn_hair_base
  bone2Name=head
  swing1LimitDegrees=45
  swing1Motion=Locked
  swing2LimitDegrees=45
  swing2Motion=Locked
  twistLimitDegrees=45
  twistMotion=Locked
 }
 {bone1Name=dyn_hair_12_c, bone2Name=dyn_hair_11_c, swing1LimitDegrees=20, swing2LimitDegrees=20, twistLimitDegrees=15}
 {bone1Name=dyn_hair_01_c, bone2Name=dyn_hair_base, swing1LimitDegrees=10, swing2LimitDegrees=35, twistLimitDegrees=5}
 {bone1Name=dyn_hair_02_c, bone2Name=dyn_hair_01_c, swing1LimitDegrees=15, swing2LimitDegrees=15, twistLimitDegrees=5}
 {bone1Name=dyn_hair_03_c, bone2Name=dyn_hair_02_c, swing1LimitDegrees=20, swing2LimitDegrees=20, twistLimitDegrees=20}
 {bone1Name=dyn_hair_04_c, bone2Name=dyn_hair_03_c, swing1LimitDegrees=20, swing2LimitDegrees=20, twistLimitDegrees=5}
 {bone1Name=dyn_hair_01_r, bone2Name=dyn_hair_04_c, swing1LimitDegrees=10, swing2LimitDegrees=10, twistLimitDegrees=5}
 {bone1Name=dyn_hair_05_c, bone2Name=dyn_hair_04_c, swing1LimitDegrees=20, swing2LimitDegrees=25, twistLimitDegrees=5}
 {bone1Name=dyn_hair_01_l, bone2Name=dyn_hair_05_c, swing1LimitDegrees=10, swing2LimitDegrees=15, twistLimitDegrees=5}
 {bone1Name=dyn_hair_06_c, bone2Name=dyn_hair_05_c, swing1LimitDegrees=20, swing2LimitDegrees=20, twistLimitDegrees=5}
 {bone1Name=dyn_hair_07_c, bone2Name=dyn_hair_06_c, swing1LimitDegrees=35, swing2LimitDegrees=45, twistLimitDegrees=10}
 {bone1Name=dyn_hair_02_l, bone2Name=dyn_hair_07_c, swing1LimitDegrees=20, swing2LimitDegrees=35, twistLimitDegrees=5}
 {bone1Name=dyn_hair_02_r, bone2Name=dyn_hair_07_c, swing1LimitDegrees=10, swing2LimitDegrees=10, twistLimitDegrees=5}
 {bone1Name=dyn_hair_03_r, bone2Name=dyn_hair_02_r, swing1LimitDegrees=10, swing2LimitDegrees=10, twistLimitDegrees=5}
 {bone1Name=dyn_hair_04_r, bone2Name=dyn_hair_03_r, swing1LimitDegrees=15, swing2LimitDegrees=15, twistLimitDegrees=5}
 {bone1Name=dyn_hair_08_c, bone2Name=dyn_hair_07_c, swing1LimitDegrees=20, swing2LimitDegrees=20, twistLimitDegrees=10}
 {bone1Name=dyn_hair_09_c, bone2Name=dyn_hair_08_c, swing1LimitDegrees=20, swing2LimitDegrees=20, twistLimitDegrees=10}
 {bone1Name=dyn_hair_10_c, bone2Name=dyn_hair_09_c, swing1LimitDegrees=20, swing2LimitDegrees=20, twistLimitDegrees=10}
 {bone1Name=dyn_hair_11_c, bone2Name=dyn_hair_10_c, swing1LimitDegrees=20, swing2LimitDegrees=20, twistLimitDegrees=10}
 {
  bone1Name=head
  bone2Name=spine_05
  swing1LimitDegrees=45
  swing1Motion=Free
  swing2LimitDegrees=45
  swing2Motion=Free
  twistLimitDegrees=45
  twistMotion=Free
 }
 {
  bone1Name=spine_02
  bone2Name=spine_01
  swing1LimitDegrees=45
  swing1Motion=Free
  swing2LimitDegrees=45
  swing2Motion=Free
  twistLimitDegrees=45
  twistMotion=Free
 }
 {
  bone1Name=spine_03
  bone2Name=spine_02
  swing1LimitDegrees=45
  swing1Motion=Free
  swing2LimitDegrees=45
  swing2Motion=Free
  twistLimitDegrees=45
  twistMotion=Free
 }
 {
  bone1Name=spine_04
  bone2Name=spine_03
  swing1LimitDegrees=45
  swing1Motion=Free
  swing2LimitDegrees=45
  swing2Motion=Free
  twistLimitDegrees=45
  twistMotion=Free
 }
 {
  bone1Name=spine_05
  bone2Name=spine_04
  swing1LimitDegrees=45
  swing1Motion=Free
  swing2LimitDegrees=45
  swing2Motion=Free
  twistLimitDegrees=45
  twistMotion=Free
 }
 {
  bone1Name=calf_l
  bone2Name=thigh_l
  swing1LimitDegrees=45
  swing1Motion=Free
  swing2LimitDegrees=45
  swing2Motion=Free
  twistLimitDegrees=45
  twistMotion=Free
 }
 {
  bone1Name=calf_r
  bone2Name=thigh_r
  swing1LimitDegrees=45
  swing1Motion=Free
  swing2LimitDegrees=45
  swing2Motion=Free
  twistLimitDegrees=45
  twistMotion=Free
 }
 {
  bone1Name=upperarm_r
  bone2Name=spine_05
  swing1LimitDegrees=45
  swing1Motion=Free
  swing2LimitDegrees=45
  swing2Motion=Free
  twistLimitDegrees=45
  twistMotion=Free
 }
 {
  bone1Name=upperarm_l
  bone2Name=spine_05
  swing1LimitDegrees=45
  swing1Motion=Free
  swing2LimitDegrees=45
  swing2Motion=Free
  twistLimitDegrees=45
  twistMotion=Free
 }
 {
  bone1Name=spine_04
  bone2Name=dyn_hair_04_c
  swing1LimitDegrees=45
  swing1Motion=Free
  swing2LimitDegrees=45
  swing2Motion=Free
  twistLimitDegrees=45
  twistMotion=Free
 }
 {
  bone1Name=spine_05
  bone2Name=dyn_hair_03_c
  swing1LimitDegrees=45
  swing1Motion=Free
  swing2LimitDegrees=45
  swing2Motion=Free
  twistLimitDegrees=45
  twistMotion=Free
 }
 {
  bone1Name=spine_03
  bone2Name=dyn_hair_05_c
  swing1LimitDegrees=45
  swing1Motion=Free
  swing2LimitDegrees=45
  swing2Motion=Free
  twistLimitDegrees=45
  twistMotion=Free
 }
 {
  bone1Name=spine_05
  bone2Name=dyn_hair_02_c
  swing1LimitDegrees=45
  swing1Motion=Free
  swing2LimitDegrees=45
  swing2Motion=Free
  twistLimitDegrees=45
  twistMotion=Free
 }
 {
  bone1Name=spine_01
  bone2Name=pelvis
  swing1LimitDegrees=45
  swing1Motion=Free
  swing2LimitDegrees=45
  swing2Motion=Free
  twistLimitDegrees=45
  twistMotion=Free
 }
 {
  bone1Name=thigh_l
  bone2Name=pelvis
  swing1LimitDegrees=45
  swing1Motion=Free
  swing2LimitDegrees=45
  swing2Motion=Free
  twistLimitDegrees=45
  twistMotion=Free
 }
 {
  bone1Name=thigh_r
  bone2Name=pelvis
  swing1LimitDegrees=45
  swing1Motion=Free
  swing2LimitDegrees=45
  swing2Motion=Free
  twistLimitDegrees=45
  twistMotion=Free
 }
 {bone1Name=dyn_hair_11_c, bone2Name=dyn_hair_04_r, swing1LimitDegrees=15, swing2LimitDegrees=15, twistLimitDegrees=15}
 {bone1Name=dyn_hair_10_c, bone2Name=dyn_hair_03_r, swing1LimitDegrees=10, swing2LimitDegrees=10, twistLimitDegrees=10}
]
