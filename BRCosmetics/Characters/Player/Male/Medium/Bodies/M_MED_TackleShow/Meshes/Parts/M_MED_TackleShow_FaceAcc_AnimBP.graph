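# Animation Blueprint for the TackleShow face accessory part, adding secondary physics that adapts to a worn cape or backbling.
# It blends attached-parent pose with Body thigh bones, then runs a Control Rig and ground-collider bone modifier into a cached pose.
# That feeds three RigidBody sims with separate physics assets, selected by cape and backbling flags.
# The event graph updates those flags, a gravity override, and a ground-collider offset lowered when airborne, emoting or in vehicles.
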AnimBlueprint /BRCosmetics/Characters/Player/Male/Medium/Bodies/M_MED_TackleShow/Meshes/Parts/M_MED_TackleShow_FaceAcc_AnimBP
target_skeleton=&/M_MED_TackleShow_FaceAcc_Skeleton
graphs [
 AnimGraph {nodes [
   RigidBody!(ComponentPose, Alpha, OverrideWorldGravity:Vector="0,0,-980", ExternalForce) -> (Pose) {
    bClampLinearTranslationLimitToRefPose=true
    bOverrideWorldGravity=true
    overridePhysicsAsset=&/M_MED_TackleShow_FaceAcc_Physics
    simulationSpace=BaseBoneSpace
    baseBoneRef {boneName=root}
    overrideWorldGravity {z=-980}
    simSpaceSettings {
     maxAngularAcceleration=85
     maxAngularVelocity=3
     maxLinearAcceleration=155
     maxLinearVelocity=140
     worldAlpha=1
     externalAngularVelocity=<0.2 0.2 0.4>
     externalLinearDragV=<0.1 0.1 0.2>
     externalLinearVelocity=<0.2 0.2 0.2>
    }
   }
   n1 = RigidBody(ComponentPose=Local To Component(LocalPose=Use cached pose 'Pre Sim 01'()), OverrideWorldGravity="0,0,-980") {
    bClampLinearTranslationLimitToRefPose=true
    bOverrideWorldGravity=true
    overridePhysicsAsset=&M_MED_TackleShow_FaceAcc_Physics
    simulationSpace=BaseBoneSpace
    baseBoneRef {boneName=root}
    overrideWorldGravity {z=-980}
    simSpaceSettings {
     maxAngularAcceleration=40
     maxAngularVelocity=10
     maxLinearAcceleration=1000
     maxLinearVelocity=200
     worldAlpha=1
     externalAngularVelocity=<0.1 0.1 0.1>
     externalLinearDragV=<0.1 0.1 0.1>
     externalLinearVelocity=<0.1 0.1 0.1>
    }
   }
   n2 = Component To Local(ComponentPose=@n1)
   n3 = RigidBody(ComponentPose=Local To Component(LocalPose=Use cached pose 'Pre Sim 01'()), OverrideWorldGravity="0,0,-980") {
    bClampLinearTranslationLimitToRefPose=true
    bOverrideWorldGravity=true
    overridePhysicsAsset=&/M_MED_TackleShow_FaceAcc_BackBling_Physics
    simulationSpace=BaseBoneSpace
    baseBoneRef {boneName=root}
    overrideWorldGravity {z=-980}
    simSpaceSettings {
     maxAngularAcceleration=20
     maxAngularVelocity=4
     maxLinearAcceleration=500
     maxLinearVelocity=150
     worldAlpha=1
     externalAngularVelocity=<0.05 0.05 0.05>
     externalLinearDragV=<0.05 0.05 0.05>
     externalLinearVelocity=<0.05 0.05 0.05>
    }
   }
   n4 = Component To Local(ComponentPose=@n3)
   n5 = Get Wear_Backbling() -> (Wear_Backbling:bool)
   n6 = RigidBody(ComponentPose=Local To Component(LocalPose=Use cached pose 'Pre Sim 01'()), OverrideWorldGravity="0,0,-980") {
    bClampLinearTranslationLimitToRefPose=true
    bOverrideWorldGravity=true
    overridePhysicsAsset=&/M_MED_TackleShow_FaceAcc_Cape_Physics
    simulationSpace=BaseBoneSpace
    baseBoneRef {boneName=dyn_hair_06_c}
    overrideWorldGravity {z=-980}
    simSpaceSettings {
     maxAngularAcceleration=50
     maxAngularVelocity=10
     maxLinearAcceleration=200
     maxLinearVelocity=200
     worldAlpha=1
     externalAngularVelocity=<0.2 0.2 0.2>
     externalLinearDragV=<0.2 0.2 0.2>
     externalLinearVelocity=<0.2 0.2 0.2>
    }
   }
   Output Pose(Result=Blend Poses by bool(
     BlendPose_0=Component To Local(ComponentPose=@n6)
     BlendPose_1=Blend Poses by bool(BlendPose_0=@n4, BlendPose_1=@n2, bActiveValue=@n5) {blendType=Linear}
     bActiveValue=Get Wear_cape() -> (Wear_cape:bool)
    ) {blendType=Linear}) {layerGroup=None}
   n7 = Get Body() -> (Body:Skeletal Mesh Component Budgeted Object Reference)
   n8 = Control Rig!(
    Source=Layered blend per bone(
     BasePose=Copy Pose From Mesh() {bUseAttachedParent=true}
     BlendPoses_0=Copy Pose From Mesh(SourceMeshComponent=@n7) {bUseAttachedParent=true}
    ) {layerSetup [{branchFilters [{blendDepth=1, boneName=thigh_r}, {blendDepth=1, boneName=thigh_l}]}]}
    Alpha
    Backbling:bool=Get Wear_Backbling()
    WearCape:bool=Get Wear_cape()
   ) -> (Pose) {controlRigAssetReference {controlRigAsset=&/M_MED_TackleShow_FaceAcc_CtrlRig}}
   n9 = Local To Component(LocalPose=@n8)
   Save cached pose 'Pre Sim 01'(Pose=Component To Local(ComponentPose="Transform (Modify) Bone - Bone: deform_ground_collider"(ComponentPose=@n9, Alpha=1, Translation="0,0,0", Rotation="0,0,0", Scale="1,1,1") {translationMode=BMM_Additive, boneToModify {boneName=deform_ground_collider}})) {cacheName=Pre Sim 01}
  ]}
 EventGraph {nodes [
   Event BlueprintUpdateAnimation()
   > n1 = Sequence!(execute) -> (then_0, then_1, then_2:Exec, then_3:Exec, then_4:Exec)
   > n2 = "Gravity Override State Function (Parameter Asset) (Nativized)"(
    CharacterPartAnimInstance=Self-Reference()
    GameTimeInSeconds=GetGameTimeInSeconds()
    Parameters=/BRCosmetics/Characters/Player/Male/Medium/Bodies/M_MED_TackleShow/Meshes/M_MED_TackleShow_Parameters
   )
   > Set GravityOverride(GravityOverride=@n2)
   n3 = CheckForBackbling(self=Self-Reference())
   Set Wear_cape(execute=@n1.then_1, Wear_cape:bool=@n3.bIsBackblingACape) -> (Output_Get:bool)
   Set Wear_Backbling(execute=@n1.then_2, Wear_Backbling:bool=@n3) -> (Output_Get:bool)
   n4 = Get bIsOnGround() -> (bIsOnGround:bool)
   n5 = Branch(
    execute=@n1.then_3
    Condition=OR Boolean!(A=NOT Boolean(A=@n4), B=Get bIsPlayingEmote(), C:bool=Get bIsInVehicle() -> (bIsInVehicle:bool)) -> (ReturnValue)
   )
   > Set Ground Collider(Ground Collider:Vector="0,0,-2000") -> (Output_Get:Vector)
   Set Ground Collider(execute=@n5.else, Ground Collider="0, 0, 0")
   n6 = Event OnCharacterCustomizationLoaded()
   > Set Body(Body:Skeletal Mesh Component Budgeted Object Reference=Get Array Item with EFort Custom Part Type(Array=@n6.CustomMeshParts)) -> (Output_Get:Skeletal Mesh Component Budgeted Object Reference)
   Event BlueprintInitializeAnimation()
   > Set SkelMesh(SkelMesh=GetOwnerSkeletalMeshComponent())
   n7 = TryGetPawnOwner()
   > Is Valid(InputObject=@n7)
   > n8 = Cast To FortPlayerPawnAthena(Object=@n7)
   > Set PlayerPawn(PlayerPawn:Fort Player Pawn Athena Object Reference=@n8) -> (Output_Get:Fort Player Pawn Athena Object Reference)
  ]}
]
variables [
 Body {type=SkeletalMeshComponentBudgeted}
 Gravity Override Params {
  type=GravityOverrideParamsStruct
  default {
   dBNO {jointName=root, gravityOverride=<0 -450 -980>}
   skydive_DivingDown {windAmplitude=<50 100 100>}
   skydive_FloatingDown {gravityOverride=<0 -200 300>, windAmplitude=<150 -100 150>, windFrequency=<1.5 10 12>, windOffset=<0 -500 70>}
   skydive_GoingUp {gravityOverride=<0 0 -400>, windAmplitude=<50 -50 50>, windOffset=<0 -600 50>}
   skydive_Parachuting {gravityOverride=<0 -500 -980>, windAmplitude=<100 -100 100>}
  }
 }
 GravityOverride {type=Vector}
 Ground Collider {type=Vector}
 PlayerPawn {type=FortPlayerPawnAthena}
 RBAN Params {
  type=RBANParamsStruct
  default {
   dBNO {simSpaceSettings {
     maxAngularAcceleration=1000
     maxAngularVelocity=3
     maxLinearAcceleration=300
     maxLinearVelocity=700
     velocityScaleZ=0
     externalLinearDragV=<0 0.5 0>
    }}
   dBNO_Carried {simSpaceSettings {
     maxAngularAcceleration=1000
     maxAngularVelocity=3
     maxLinearAcceleration=300
     maxLinearVelocity=700
     velocityScaleZ=0
     externalLinearDragV=<0 0.5 0>
    }}
   emoteOrMelee {simSpaceSettings {
     maxAngularAcceleration=1000
     maxAngularVelocity=3
     maxLinearAcceleration=300
     maxLinearVelocity=700
     velocityScaleZ=0
     externalLinearDragV=<0 0.5 0>
    }}
   front_End {simSpaceSettings {
     maxAngularAcceleration=1000
     maxAngularVelocity=3
     maxLinearAcceleration=300
     maxLinearVelocity=700
     velocityScaleZ=0
     externalLinearDragV=<0 0.5 0>
    }}
   onGround_Crouch_Moving {simSpaceSettings {
     maxAngularAcceleration=1000
     maxAngularVelocity=3
     maxLinearAcceleration=300
     maxLinearVelocity=700
     velocityScaleZ=0
     externalLinearDragV=<0 0.5 0>
    }}
   onGround_Falling {simSpaceSettings {
     maxAngularAcceleration=1000
     maxAngularVelocity=3
     maxLinearAcceleration=300
     maxLinearVelocity=700
     velocityScaleZ=0
     externalLinearDragV=<0 0.5 0>
    }}
   onGround_Moving {simSpaceSettings {
     maxAngularAcceleration=1000
     maxAngularVelocity=3
     maxLinearAcceleration=300
     maxLinearVelocity=700
     velocityScaleZ=0
     externalLinearDragV=<0 0.5 0>
    }}
   onGround_Standing {simSpaceSettings {
     maxAngularAcceleration=1000
     maxAngularVelocity=3
     maxLinearAcceleration=300
     maxLinearVelocity=700
     velocityScaleZ=0
     externalLinearDragV=<0 0.5 0>
    }}
   rail_Grinding {simSpaceSettings {
     maxAngularAcceleration=1000
     maxAngularVelocity=3
     maxLinearAcceleration=300
     maxLinearVelocity=700
     velocityScaleZ=0
     worldAlpha=0.1
     externalLinearDragV=<0 0.5 0>
    }}
   riding_Moving {simSpaceSettings {
     maxAngularAcceleration=1000
     maxAngularVelocity=3
     maxLinearAcceleration=300
     maxLinearVelocity=700
     velocityScaleZ=0
     externalLinearDragV=<0 0.5 0>
    }}
   riding_Moving_Targeting {simSpaceSettings {
     maxAngularAcceleration=1000
     maxAngularVelocity=3
     maxLinearAcceleration=300
     maxLinearVelocity=700
     velocityScaleZ=0
     externalLinearDragV=<0 0.5 0>
    }}
   riding_Standing {simSpaceSettings {
     maxAngularAcceleration=1000
     maxAngularVelocity=3
     maxLinearAcceleration=300
     maxLinearVelocity=700
     velocityScaleZ=0
     externalLinearDragV=<0 0.5 0>
    }}
   skydive_DivingDown {simSpaceSettings {
     maxAngularAcceleration=1000
     maxAngularVelocity=3
     maxLinearAcceleration=300
     maxLinearVelocity=700
     velocityScaleZ=0
     externalLinearDragV=<0 0.5 0>
    }}
   skydive_FloatingDown {
    componentLinearAccScale=<0 0.5 0>
    componentLinearVelScale=<0 0.5 0>
    simSpaceSettings {
     maxAngularAcceleration=1000
     maxAngularVelocity=3
     maxLinearAcceleration=300
     maxLinearVelocity=700
     velocityScaleZ=0
     externalLinearDragV=<0 0.5 0>
    }
   }
   skydive_GoingUp {simSpaceSettings {
     maxAngularAcceleration=1000
     maxAngularVelocity=3
     maxLinearAcceleration=300
     maxLinearVelocity=700
     velocityScaleZ=0
     externalLinearDragV=<0 0.5 0>
    }}
   skydive_Parachuting {
    componentLinearAccScale=<0 0.1 0.3>
    componentLinearVelScale=<0 0.1 0.3>
    simSpaceSettings {
     maxAngularAcceleration=1000
     maxAngularVelocity=3
     maxLinearAcceleration=300
     maxLinearVelocity=700
     velocityScaleZ=0
     externalLinearDragV=<0 0.5 0>
    }
   }
   skydive_Surfing {simSpaceSettings {
     maxAngularAcceleration=1000
     maxAngularVelocity=3
     maxLinearAcceleration=300
     maxLinearVelocity=700
     velocityScaleZ=0
     externalLinearDragV=<0 0.5 0>
    }}
   swimming {simSpaceSettings {
     maxAngularAcceleration=1000
     maxAngularVelocity=3
     maxLinearAcceleration=300
     maxLinearVelocity=700
     velocityScaleZ=0
     externalLinearDragV=<0 0.5 0>
    }}
   tactical_Sprint {simSpaceSettings {
     maxAngularAcceleration=1000
     maxAngularVelocity=3
     maxLinearAcceleration=300
     maxLinearVelocity=700
     velocityScaleZ=0
     externalLinearDragV=<0 0.5 0>
    }}
   ziplining {simSpaceSettings {
     maxAngularAcceleration=1000
     maxAngularVelocity=3
     maxLinearAcceleration=300
     maxLinearVelocity=700
     velocityScaleZ=0
     externalLinearDragV=<0 0.5 0>
    }}
  }
 }
 SkelMesh {type=SkeletalMeshComponent}
 Skydive {type=boolean}
 Wear_Backbling {type=boolean}
 Wear_cape {type=boolean}
]
